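# Sound cue for the end of searching an armoire loot container, played with prop-search attenuation and limited concurrency.
# A quality-level node selects one of four branches.
# The first two mix a "no more loot" UI sting with random armoire search foley, volume-enveloped into a short burst and pitch/volume modulated; the second has fewer variants.
# The last two play only one modulated, apparently optimized SearchEnd wave as cheaper alternatives.
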SoundCue /Game/Sounds/Foley_Loot/Containers/Armoire/Armoire_SearchEnd_Cue
root=QualityLevel_0
nodes [
 {
  id=Enveloper_0
  children [Random_0]
  properties {
   pitchCurve {editorCurveData {keys [{time=0}]}}
   volumeCurve {editorCurveData {keys [<at 0.0045 0.0028>,<at 0.043 0.4428>,<at 0.4967 0.0028>]}}
  }
 }
 {
  id=Enveloper_20
  children [Random_48]
  properties {volumeCurve {editorCurveData {keys [<at 0.0045 0.0028>,<at 0.043 0.4428>,<at 0.4967 0.0028>]}}}
 }
 {id=Mixer_0, children [WavePlayer_1,Modulator_0], properties {inputVolume [0.8,1.1]}}
 {id=Mixer_12, children [WavePlayer_43, Modulator_15], properties {inputVolume [0.8,1.1]}}
 {id=Modulator_0, children [Enveloper_0], properties {pitchMax=1.07, pitchMin=0.92, volumeMax=1.07, volumeMin=0.92}}
 {id=Modulator_1, children [WavePlayer_0], properties {pitchMax=1.07, pitchMin=0.92, volumeMax=1.07, volumeMin=0.92}}
 {id=Modulator_15, children [Enveloper_20], properties {pitchMax=1.07, pitchMin=0.92, volumeMax=1.07, volumeMin=0.92}}
 {id=Modulator_2, children [WavePlayer_5], properties {pitchMax=1.07, pitchMin=0.92, volumeMax=1.07, volumeMin=0.92}}
 {id=QualityLevel_0, children [Mixer_12,Mixer_0,Modulator_1,Modulator_2]}
 {id=Random_0, children [WavePlayer_2,WavePlayer_3], properties {weights [1,1]}}
 {id=Random_48, children [WavePlayer_302, WavePlayer_307, WavePlayer_308], properties {weights [1,1,1]}}
 {id=WavePlayer_0, properties {soundWaveAssetPtr=&/Armoire_SearchEnd_opt}}
 {id=WavePlayer_1, properties {soundWaveAssetPtr=/Game/Sounds/Fort_UI_Sounds/Gameplay/Fort_UI_NoMoreLoot}}
 {id=WavePlayer_2, properties {soundWaveAssetPtr=&/Armoire_Search07}}
 {id=WavePlayer_3, properties {soundWaveAssetPtr=&/Armoire_Search05}}
 {id=WavePlayer_302, properties {soundWaveAssetPtr=&Armoire_Search07}}
 {id=WavePlayer_307, properties {soundWaveAssetPtr=&Armoire_Search05}}
 {id=WavePlayer_308, properties {soundWaveAssetPtr=&/Armoire_Search06}}
 {id=WavePlayer_43, properties {soundWaveAssetPtr=&Fort_UI_NoMoreLoot}}
 {id=WavePlayer_5, properties {soundWaveAssetPtr=&Armoire_SearchEnd_opt}}
]
properties {
 attenuationSettings=/Game/Sounds/Attenuation/Fort_Prop_Search_Att
 bOverrideConcurrency=true
 maxDistance=1300
 soundClassObject=&Fort_Loot_Collect
 volumeMultiplier=0.6
 concurrencyOverrides {maxCount=4, resolutionRule=StopFarthestThenPreventNew}
}
