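# Sound cue for the foley heard when an armoire loot container is searched.
# A quality-level node picks between two layered mixes of random armoire and gear-search waves, one with three variants per layer and one with two, or two branches playing a single optimized wave.
# Modulators add pitch and volume variation.
# It uses prop-search attenuation, the loot-collect sound class and a concurrency override that prevents new instances at the limit.
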SoundCue /Game/Sounds/Foley_Loot/Containers/Armoire/Armoire_Search_Cue
root=QualityLevel_0
nodes [
 {id=Mixer_0, children [Modulator_0,Modulator_1], properties {inputVolume [0.3,1.2]}}
 {id=Mixer_11, children [Modulator_23,Random_10], properties {inputVolume [0.3,1.2]}}
 {id=Modulator_0, children [Random_0], properties {pitchMax=1.07, pitchMin=0.92, volumeMax=1.07, volumeMin=0.92}}
 {id=Modulator_1, children [Random_1], properties {pitchMax=1.07, pitchMin=0.92, volumeMax=1.07, volumeMin=0.92}}
 {id=Modulator_23, children [Random_49], properties {pitchMax=1.07, pitchMin=0.92, volumeMax=1.07, volumeMin=0.92}}
 {id=Modulator_3, children [WavePlayer_8]}
 {id=Modulator_4, children [WavePlayer_9]}
 {id=QualityLevel_0, children [Mixer_11,Mixer_0,Modulator_3,Modulator_4]}
 {id=Random_0, children [WavePlayer_0,WavePlayer_1], properties {weights [1,1]}}
 {id=Random_1, children [WavePlayer_3,WavePlayer_4], properties {weights [1,1]}}
 {id=Random_10, children [WavePlayer_47, WavePlayer_49, WavePlayer_51], properties {weights [1,1,1]}}
 {id=Random_49, children [WavePlayer_316, WavePlayer_321, WavePlayer_322], properties {weights [1,1,1]}}
 {id=WavePlayer_0, properties {soundWaveAssetPtr=&/Armoire_Search07}}
 {id=WavePlayer_1, properties {soundWaveAssetPtr=&/Armoire_Search05}}
 {id=WavePlayer_3, properties {soundWaveAssetPtr=/Game/Sounds/Foley_Loot/Containers/Misc/Prop_Gear_Search_05}}
 {id=WavePlayer_316, properties {soundWaveAssetPtr=&Armoire_Search07}}
 {id=WavePlayer_321, properties {soundWaveAssetPtr=&Armoire_Search05}}
 {id=WavePlayer_322, properties {soundWaveAssetPtr=&/Armoire_Search06}}
 {id=WavePlayer_4, properties {soundWaveAssetPtr=/Game/Sounds/Foley_Loot/Containers/Misc/Prop_Gear_Search_01}}
 {id=WavePlayer_47, properties {soundWaveAssetPtr=&Prop_Gear_Search_05}}
 {id=WavePlayer_49, properties {soundWaveAssetPtr=&Prop_Gear_Search_01}}
 {id=WavePlayer_51, properties {soundWaveAssetPtr=/Game/Sounds/Foley_Loot/Containers/Misc/Prop_Gear_Search_06}}
 {id=WavePlayer_8, properties {soundWaveAssetPtr=&/Armoire_Search_opt}}
 {id=WavePlayer_9, properties {soundWaveAssetPtr=&Armoire_Search_opt}}
]
properties {
 attenuationSettings=/Game/Sounds/Attenuation/Fort_Prop_Search_Att
 bOverrideConcurrency=true
 maxDistance=1300
 soundClassObject=&Fort_Loot_Collect
 volumeMultiplier=0.85
 concurrencyOverrides {resolutionRule=PreventNew}
}
